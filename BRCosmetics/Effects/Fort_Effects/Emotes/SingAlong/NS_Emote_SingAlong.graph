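# Niagara system for the SingAlong emote, a single sprite emitter that fills a raised spherical volume with small, short-lived particles.
# Spawn rate ramps up along a curve over emitter age during the three-second loop duration.
# Particles get slight random horizontal velocity, then sink under soft gravity while vector noise jostles them and sprites spin.
# Curves fade and scale each sprite in quickly and out over its life; Drag is disabled.
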NiagaraSystem /BRCosmetics/Effects/Fort_Effects/Emotes/SingAlong/NS_Emote_SingAlong
dependencies {usedDataInterfaces [&NiagaraDataInterfaceCurve], usedDynamicInputs [], usedModules [&SystemState]}
emitters {NE_Empty {
  emitterUpdateScript [
   &EmitterState(Inactive Response=::NewEnumerator1, Life Cycle Mode=::NewEnumerator1, Loop Duration=3, UseLoopDelay=false)
   &SpawnRate(
    SpawnRate=&Multiply_Float(
     A=220
     B=&FloatFromCurve(FloatCurve=DataInterface() {propertyValues {Curve {keys [<at 0 0>,<at 2.4 1>]}}}, CurveIndex=$Emitter.Age)
    )
    Use Spawn Probability=false
   )
  ]
  particleSpawnScript [
   /Niagara/Modules/Spawn/Initialization/InitializeParticle(
    Lifetime=&RandomRangeFloat(Minimum=1.5, Maximum=3)
    Module.Write Color=true
    Module.Write Lifetime=true
    Module.Write Mass=true
    Module.Write Material Random=true
    Module.Write Position=false
    Module.Write Scale=false
    Module.Write SpriteRotation=true
    Module.Write SpriteSize=true
    Position=&SimulationPosition()
    Sprite Rotation=&RandomRangeFloat(Minimum=-180, Maximum=180)
    Sprite Size=&Vector2DFromFloat(Value=&RandomRangeFloat(Minimum=1, Maximum=2))
   )
   &SystemLocation(MaskSpawnGroup=false, Offset=<0 0 165>)
   &SphereLocation(
    MaskSpawnGroup=false
    Override Seed=false
    Sphere Coordinate Space=::Simulation
    Sphere Radius=170
    Surface Only=false
    UseNonUniformScale=false
    UseOffset=false
    UseSphereOrientation=false
   )
   &AddVelocity(Velocity=&RandomRangeVector(Minimum=<-15 -15 0>, Maximum=<15 15 0>))
  ]
  particleUpdateScript [
   &ParticleState()
   &GravityForce(Gravity=&RandomRangeVector(Minimum=<0 0 -35>, Maximum=<1 1 -45>))
   &VectorNoiseForce(Force Amount=800, Module.Use Random Noise Time Interval=true, Random Noise Time Interval=1)
   &Drag(Use Linear Drag=true, Use Rotational Drag=false) {enabled=false}
   &SolveForcesAndVelocity(Clamp Velocity=false, Limit Acceleration=false)
   &SpriteRotationRate(Rotation Rate=&RandomRangeFloat(Minimum=-200, Maximum=200))
   &ScaleColor(
    Scale Alpha=&FloatFromCurve(FloatCurve=DataInterface() {propertyValues {Curve {keys [
         {arriveTangent=7.7613, interpMode=RCIM_Cubic, leaveTangent=7.7613, tangentMode=RCTM_User, time=0, value=0}
         <at 0.2315 1 RCIM_Cubic>
         <at 1 0 RCIM_Cubic>
        ]}}})
    ScaleA=true
    ScaleRGB=true
   )
   &ScaleSpriteSize(Scale Factor=&Vector2DFromFloat(Value=&FloatFromCurve(FloatCurve=DataInterface() {propertyValues {Curve {keys [
          {arriveTangent=10.741, interpMode=RCIM_Cubic, leaveTangent=10.741, tangentMode=RCTM_User, time=0, value=0}
          <at 0.1854 0.984 RCIM_Cubic>
          {arriveTangent=-2.6177, interpMode=RCIM_Cubic, leaveTangent=-2.6177, tangentMode=RCTM_User, time=1, value=0}
         ]}}})))
  ]
  rendererClasses [&NiagaraSpriteRendererProperties]
  renderers [{rendererClass=&NiagaraSpriteRendererProperties}]
 }}
summary {
 systemName=NS_Emote_SingAlong
 emitters [{emitterName=NE_Empty, rendererClasses [&NiagaraSpriteRendererProperties]}]
 userVariables []
}
